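; PCG subgraph that creates a single point representing an actor's bounds.
; Get Actor Data runs in single-point mode to produce the point, which feeds Apply Scale To Bounds before reaching the output.
; Baking actor scale into the bounds lets downstream nodes rely on pre-multiplied extents.
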(graph Create_ActorBounds_CoreProcess
  (bind GetActorData_1 (Get%20Actor%20Data :Mode GetSinglePoint)) ; Get self as a single point
  (bind ApplyScaleToBounds_1 (Apply%20Scale%20To%20Bounds)) ; Apply actor scale to bounds to rely on pre-multiplied extents downstream
  (connect GetActorData_1 Out ApplyScaleToBounds_1 In)
  (output Out ApplyScaleToBounds_1))
