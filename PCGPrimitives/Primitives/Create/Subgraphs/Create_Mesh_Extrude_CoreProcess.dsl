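; PCG subgraph that extrudes input splines into one merged, material-assigned mesh, each shape with its own height.
; Splines are rebuilt closed, given Data.Height from a validated data-level HeightAttribute or fallback Min/Max noise, then forced counter-clockwise.
; A loop over the BuildSides subgraph extrudes each spline before merging.
; DynamicMesh selects direct dynamic mesh spawning or saving an actor-offset asset for static mesh spawning.
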(graph Create_Mesh_Extrude_CoreProcess
  (params (MinHeight :type Double :default 1e+03) (MaxHeight :type Double :default 5e+03) (OutMeshPath :type String) (OutMeshName :type String) (DynamicMesh :type Bool :default true) (Material :type (SoftObject /Script/Engine.MaterialInterface)) (HeightAttribute :type String))
  (bind FilterDataByType_0 (Filter%20Data%20By%20Type :TargetType (struct :Ids [(struct :Struct /Script/PCG.PCGDataTypeInfoPolyline)])))
  (bind SplineDirection_3 (Spline%20Direction :Operation ForceCounterClockwise)) ; Force winding order for extrusion
  (bind SetMaterialsDynamicMesh_4 (Set%20Materials%20On%20Dynamic%20Mesh :Materials-param Material))
  (bind Branch_0 (Branch :bOutputToB-param DynamicMesh)) ; Whether the resulting mesh is saved as an asset before spawn\nor directly spawned as a dynamic mesh.
  ;[0 0 0 0.5] Manage @Data.Height attribute required for this operation
  (bind AttributeMathsOp_1 (Attribute%20Maths%20Op :Operation Subtract :InA MaxHeight :InB MinHeight)) ; Min/Max average
  (bind AddAttribute_2 (Add%20Attribute :AttributeTypes (struct :Type Float :FloatValue 5e+03 :DoubleValue 7.5e+02 :VectorValue (0.0 0.0 5e+03)) :OutputTarget @Data.Height)) ; Create standardized @Data.Height attribute required for this operation
  (bind AttributeNoise_1 (Attribute%20Noise :bHasCustomSeedSource true :CustomSeedSource $LocalPosition.X :InputSource @Data.Height :NoiseMax 2.5e+03 :NoiseMax-param MaxHeight :NoiseMin-param MinHeight)) ; Noise attrribute using provided Min and Max Height parameters
  (bind CopyAttributes_0 (Copy%20Attributes :InputSource @Last :OutputTarget @data.Height :InputSource-param HeightAttribute)) ; copy to standardized @Data.Height attribute required for this operation
  ;[0 0 0 0.5] Attribute failsafe checks
  (bind AttributeStringOp_0 (Attribute%20String%20Op :Operation Substring :SearchCase IgnoreCase :InA HeightAttribute)) ; Failsafe check as this operation only supports Data level attributes values
  (bind FilterDataByAttribute_0 (Filter%20Data%20By%20Attribute :Attribute @data.MaxHeight :Attribute-param HeightAttribute)) ; Failsafe check to find if provided Height Attribute exists
  (bind Branch_3 (Branch))
  ;[0 0 0 0.5] Reset Spline Transforms
  (bind GetSplineControlPoints_0 (Get%20Spline%20Control%20Points))
  (bind CreateSpline_0 (Create%20Spline :bApplyCustomTangents true :bClosedLoop true :bUseInterpTypeAttribute true))
  ;[0 0 0 0.5] Create extruded mesh for each input data and their unique height value
  (bind Loop_0 (Loop :Graph /PCGPrimitives/Primitives/Create/Subgraphs/Create_Mesh_Extrude_CoreProcess_BuildSides.Create_Mesh_Extrude_CoreProcess_BuildSides)) ; Loop to extrude each input data according to their individual @Data.Height attribute value
  (bind MergeDynamicMeshes_1 (Merge%20Dynamic%20Meshes))
  ;[0 0 0 0.5] Apply actor offset for saved asset
  (bind GetActorData_1 (Get%20Actor%20Data :Mode GetSinglePoint))
  (bind ExtractAttribute_0 (Extract%20Attribute%20at%20Index :bForceOutputAttributeToBeInElementsDomain true :InputSource $Transform :OutputAttributeName actortransforms))
  (bind MakeTransformAttribute_0 (Make%20Transform%20Attribute :InputSource1 actortransforms.position :InputSource3 actortransforms.scale :OutputTarget actortransforms))
  ;[0 0 0 0.5] Save and spawn static mesh asset on generating actor
  (bind DynamicMeshTransform_0 (Dynamic%20Mesh%20Transform))
  (bind SaveDynamicMeshToAsset_1 (Save%20Dynamic%20Mesh%20To%20Asset :ExportParams (struct :bOpenSaveDialog false) :AssetName OutMeshName :AssetPath OutMeshPath))
  (bind CreatePoints_1 (Create%20Points))
  (bind CopyAttributes_1 (Copy%20Attributes :InputSource AssetPath :OutputTarget AssetPath))
  (bind StaticMeshSpawner_1 (Static%20Mesh%20Spawner :MeshSelectorParameters (class:/Script/PCG.PCGMeshSelectorByAttribute :AttributeName AssetPath) :MeshSelectorType /Script/PCG.PCGMeshSelectorByAttribute))
  ;[0 0 0 0.5] Spawn as dynamic mesh directly on the generating actor
  (bind SpawnDynamicMesh_0 (Spawn%20Dynamic%20Mesh))
  (connect AttributeMathsOp_1 Out AddAttribute_2 Attributes)
  (connect Branch_3 Output%20A AddAttribute_2 In)
  (connect FilterDataByAttribute_0 OutsideFilter AddAttribute_2 In)
  (connect AddAttribute_2 Out AttributeNoise_1 In)
  (connect SetMaterialsDynamicMesh_4 Out Branch_0 In)
  (connect AttributeStringOp_0 Out Branch_3 bOutputToB)
  (connect FilterDataByAttribute_0 InsideFilter Branch_3 In)
  (connect Branch_3 Output%20B CopyAttributes_0 Source)
  (connect Branch_3 Output%20B CopyAttributes_0 Target)
  (connect SaveDynamicMeshToAsset_1 OutAssetPath CopyAttributes_1 Source)
  (connect CreatePoints_1 Out CopyAttributes_1 Target)
  (connect GetSplineControlPoints_0 Out CreateSpline_0 In)
  (connect Branch_0 Output%20A DynamicMeshTransform_0 In)
  (connect MakeTransformAttribute_0 Out DynamicMeshTransform_0 Transform)
  (connect GetActorData_1 Out ExtractAttribute_0 In)
  (connect CreateSpline_0 Out FilterDataByAttribute_0 In)
  (connect DefaultInputNode In FilterDataByType_0 In)
  (connect FilterDataByType_0 InsideFilter GetSplineControlPoints_0 In)
  (connect SplineDirection_3 Out Loop_0 In)
  (connect ExtractAttribute_0 Out MakeTransformAttribute_0 Translation)
  (connect Loop_0 Out MergeDynamicMeshes_1 In)
  (connect DynamicMeshTransform_0 Out SaveDynamicMeshToAsset_1 In)
  (connect MergeDynamicMeshes_1 Out SetMaterialsDynamicMesh_4 In)
  (connect Branch_0 Output%20B SpawnDynamicMesh_0 In)
  (connect AttributeNoise_1 Out SplineDirection_3 In)
  (connect CopyAttributes_0 Out SplineDirection_3 In)
  (connect CopyAttributes_1 Out StaticMeshSpawner_1 In))
