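; PCG subgraph that builds the sides of an extruded mesh primitive from input spline data.
; Extract Attribute at Index reads the data-level height attribute and forces the result into the elements domain.
; That value drives the Thickness of a Spline To Mesh node, which takes the same input and outputs the resulting mesh.
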(graph Create_Mesh_Extrude_CoreProcess_BuildSides
  (bind ExtractAttribute_2 (Extract%20Attribute%20at%20Index :bForceOutputAttributeToBeInElementsDomain true :InputSource @data.height))
  (bind SplineToMesh_0 (Spline%20To%20Mesh :Thickness 5765.250156))
  (connect DefaultInputNode In ExtractAttribute_2 In)
  (connect DefaultInputNode In SplineToMesh_0 In)
  (connect ExtractAttribute_2 Out SplineToMesh_0 Thickness)
  (output Out SplineToMesh_0))
