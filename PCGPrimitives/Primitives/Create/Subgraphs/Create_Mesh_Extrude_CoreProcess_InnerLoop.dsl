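; PCG subgraph that serves as the inner loop of the mesh extrude primitive, building side-face meshes from input points.
; Each point's index becomes copyIndex, then its two X-end face centers become a point pair partitioned by that index.
; A reverse-sorted copy of each pair takes Z from the data-level MaxHeight, merging with the original into a four-corner loop.
; Each loop becomes a closed linear spline, forced counterclockwise, then converted to a mesh.
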(graph Create_Mesh_Extrude_CoreProcess_InnerLoop
  (bind CopyAttributes_0 (Copy%20Attributes :InputSource $Index :OutputTarget copyIndex))
  (bind ResetPointCenter_0 (Reset%20Point%20Center :PointCenterLocation (0.0 0.5 0.5)))
  (bind ResetPointCenter_1 (Reset%20Point%20Center :PointCenterLocation (1.0 0.5 0.5)))
  (bind CreatePoints_0 (Create%20Points))
  (bind CopyPoints_0 (Copy%20Points))
  (bind CopyPoints_1 (Copy%20Points))
  (bind MergePoints_0 (Merge%20Points))
  (bind AttributePartition_0 (Attribute%20Partition :PartitionAttributeSelectors [copyIndex]))
  (bind SortAttributes_0 (Sort%20Attributes :bUseStableSort false :SortAttributes [(struct :InputSource $Index :SortMethod Descending)]))
  (bind AttributeMathsOp_2 (Attribute%20Maths%20Op :InputSource1 $Position.Z :InputSource2 @data.MaxHeight :Operation Set))
  (bind MergePoints_1 (Merge%20Points))
  (bind AttributePartition_1 (Attribute%20Partition :PartitionAttributeSelectors [copyIndex]))
  (bind CreateSpline_0 (Create%20Spline :bClosedLoop true :bLinear true :InterpTypeAttribute None))
  (bind SplineDirection_0 (Spline%20Direction :Operation ForceCounterClockwise))
  (bind SplineToMesh_0 (Spline%20To%20Mesh))
  (reroute Input%20Node%20In In)
  (connect SortAttributes_0 Out AttributeMathsOp_2 InA)
  (connect Input%20Node%20In Out AttributeMathsOp_2 InB)
  (connect MergePoints_0 Out AttributePartition_0 In)
  (connect MergePoints_1 Out AttributePartition_1 In)
  (connect Input%20Node%20In Out CopyAttributes_0 Source)
  (connect Input%20Node%20In Out CopyAttributes_0 Target)
  (connect CreatePoints_0 Out CopyPoints_0 Source)
  (connect ResetPointCenter_0 Out CopyPoints_0 Target)
  (connect CreatePoints_0 Out CopyPoints_1 Source)
  (connect ResetPointCenter_1 Out CopyPoints_1 Target)
  (connect AttributePartition_1 Out CreateSpline_0 In)
  (connect CopyPoints_0 Out MergePoints_0 In)
  (connect CopyPoints_1 Out MergePoints_0 In)
  (connect AttributeMathsOp_2 Out MergePoints_1 In)
  (connect AttributePartition_0 Out MergePoints_1 In)
  (connect CopyAttributes_0 Out ResetPointCenter_0 In)
  (connect CopyAttributes_0 Out ResetPointCenter_1 In)
  (connect AttributePartition_0 Out SortAttributes_0 In)
  (connect CreateSpline_0 Out SplineDirection_0 In)
  (connect SplineDirection_0 Out SplineToMesh_0 In)
  (output Out SplineToMesh_0))
