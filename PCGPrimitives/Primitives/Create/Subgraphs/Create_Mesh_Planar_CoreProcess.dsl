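; PCG subgraph that turns input splines into a merged planar mesh with a given material.
; It rebuilds splines from control points as closed loops, feeding Spline To Mesh and, separately, a counter-clockwise direction fix for the graph output.
; DynamicMesh either spawns the result directly or, when false, offsets it by an actor-derived transform, saves it as an asset and spawns it through Static Mesh Spawner.
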(graph Create_Mesh_Planar_CoreProcess
  (params (Material :type (SoftObject /Script/Engine.MaterialInterface)) (DynamicMesh :type Bool :default true) (OutMeshName :type String) (OutMeshPath :type String))
  (bind FilterDataByType_2 (Filter%20Data%20By%20Type :TargetType (struct :Ids [(struct :Struct /Script/PCG.PCGDataTypeInfoPolyline)])))
  (bind SplineToMesh_0 (Spline%20To%20Mesh :Thickness 1.0))
  (bind MergeDynamicMeshes_0 (Merge%20Dynamic%20Meshes))
  (bind SetMaterialsDynamicMesh_1 (Set%20Materials%20On%20Dynamic%20Mesh :Materials-param Material))
  (bind Branch_0 (Branch :bOutputToB-param DynamicMesh)) ; Whether the resulting mesh is saved as an asset before spawn\nor directly spawned as a dynamic mesh.
  (bind SplineDirection_0 (Spline%20Direction :Operation ForceCounterClockwise)) ; This enforces the correct orientation for planar mesh triangulation.
  ;[0 0 0 0.5] Reset Spline Transforms
  (bind GetSplineControlPoints_2 (Get%20Spline%20Control%20Points))
  (bind CreateSpline_4 (Create%20Spline :bApplyCustomTangents true :bClosedLoop true :bUseInterpTypeAttribute true))
  ;[0 0 0 0.5] Apply actor offset for saved asset
  (bind GetActorData_2 (Get%20Actor%20Data :Mode GetSinglePoint))
  (bind ExtractAttribute_1 (Extract%20Attribute%20at%20Index :bForceOutputAttributeToBeInElementsDomain true :InputSource $Transform :OutputAttributeName actortransforms))
  (bind MakeTransformAttribute_1 (Make%20Transform%20Attribute :InputSource1 actortransforms.position :InputSource3 actortransforms.scale :OutputTarget actortransforms))
  ;[0 0 0 0.5] Save and spawn static mesh asset on generating actor
  (bind DynamicMeshTransform_1 (Dynamic%20Mesh%20Transform))
  (bind SaveDynamicMeshToAsset_0 (Save%20Dynamic%20Mesh%20To%20Asset :ExportParams (struct :bOpenSaveDialog false) :AssetName OutMeshName :AssetPath OutMeshPath))
  (bind CreatePoints_2 (Create%20Points))
  (bind CopyAttributes_2 (Copy%20Attributes :InputSource AssetPath :OutputTarget AssetPath))
  (bind StaticMeshSpawner_2 (Static%20Mesh%20Spawner :MeshSelectorParameters (class:/Script/PCG.PCGMeshSelectorByAttribute :AttributeName AssetPath) :MeshSelectorType /Script/PCG.PCGMeshSelectorByAttribute))
  ;[0 0 0 0.5] Spawn as dynamic mesh directly on the generating actor
  (bind SpawnDynamicMesh_1 (Spawn%20Dynamic%20Mesh))
  (connect SetMaterialsDynamicMesh_1 Out Branch_0 In)
  (connect SaveDynamicMeshToAsset_0 OutAssetPath CopyAttributes_2 Source)
  (connect CreatePoints_2 Out CopyAttributes_2 Target)
  (connect GetSplineControlPoints_2 Out CreateSpline_4 In)
  (connect Branch_0 Output%20A DynamicMeshTransform_1 In)
  (connect MakeTransformAttribute_1 Out DynamicMeshTransform_1 Transform)
  (connect GetActorData_2 Out ExtractAttribute_1 In)
  (connect DefaultInputNode In FilterDataByType_2 In)
  (connect FilterDataByType_2 InsideFilter GetSplineControlPoints_2 In)
  (connect ExtractAttribute_1 Out MakeTransformAttribute_1 Translation)
  (connect SplineToMesh_0 Out MergeDynamicMeshes_0 In)
  (connect DynamicMeshTransform_1 Out SaveDynamicMeshToAsset_0 In)
  (connect MergeDynamicMeshes_0 Out SetMaterialsDynamicMesh_1 In)
  (connect Branch_0 Output%20B SpawnDynamicMesh_1 In)
  (connect CreateSpline_4 Out SplineDirection_0 In)
  (connect CreateSpline_4 Out SplineToMesh_0 In)
  (connect CopyAttributes_2 Out StaticMeshSpawner_2 In)
  (output Out SplineDirection_0))
