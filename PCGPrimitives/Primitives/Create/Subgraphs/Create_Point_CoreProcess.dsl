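; PCG core-process subgraph for a Create Point primitive, building a point from Location, Rotation, Scale and Extents parameters.
; The parameters form the created point's transform and set its half-extents.
; Non-empty input tagged "Input:0" is counted to drive a branch: without it the point outputs directly.
; Otherwise the point is copied onto that input's point data, keeping the created point's attributes, color, scale and tags.
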(graph Create_Point_CoreProcess
  (params (Location :type Vector) (Rotation :type Rotator) (Scale :type Vector :default (1.0 1.0 1.0)) (Extents :type Vector :default (5e+01 5e+01 5e+01)))
  (bind MakeTransformAttribute_0 (Make%20Transform%20Attribute :Rotation Rotation :Scale Scale :Translation Location)) ; Make offset transform attribute
  (bind CreatePoints_0 (Create%20Points))
  (bind AttributeMathsOp_0 (Attribute%20Maths%20Op :InputSource1 $Extents :Operation Set :InB Extents)) ; Set point half-extents
  (bind RemoveEmptyData_1 (Remove%20Empty%20Data))
  (bind DataNum_0 (Data%20Count))
  (bind Branch_0 (Branch)) ; If primary input data is provided branch to\ncopy created point over each input points/controlpoints
  (bind FilterDataByType_2 (Filter%20Data%20By%20Type :TargetType (struct :Ids [(struct :Struct /Script/PCG.PCGDataTypeInfoPoint)])))
  (bind CopyPoints_1 (Copy%20Points :AttributeInheritance SourceOnly :bCopyEachSourceOnEveryTarget false :ColorInheritance Source :ScaleInheritance Source :TagInheritance Source)) ; copy created point over each input points/controlpoints
  ;[0 0 0 0.5 12] Primary/Secondary Inputs
  (bind FilterDataByTag_0 (Filter%20Data%20By%20Tag :SelectedTags "Input:0"))
  (connect CreatePoints_0 Out AttributeMathsOp_0 InA)
  (connect DataNum_0 Out Branch_0 bOutputToB)
  (connect AttributeMathsOp_0 Out Branch_0 In)
  (connect Branch_0 Output%20B CopyPoints_1 Source)
  (connect FilterDataByType_2 InsideFilter CopyPoints_1 Target)
  (connect MakeTransformAttribute_0 Out CreatePoints_0 Transform)
  (connect RemoveEmptyData_1 Out DataNum_0 In)
  (connect DefaultInputNode In FilterDataByTag_0 In)
  (connect RemoveEmptyData_1 Out FilterDataByType_2 In)
  (connect FilterDataByTag_0 InsideFilter RemoveEmptyData_1 In)
  (output Out Branch_0.Output%20A)
  (output Out CopyPoints_1))
