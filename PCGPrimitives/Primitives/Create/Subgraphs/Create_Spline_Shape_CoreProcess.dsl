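; PCG subgraph that creates closed spline shapes, linear or curved, from N-gon, width, length and transform parameters.
; N-gons come from a flattened sphere ring scaled by length/width; a special square case uses grid corners and a 2D convex hull.
; The shape is copied onto each input point, or without input placed relative to the actor or world origin.
; Splines are forced counterclockwise, and squares are rebuilt from control points to fix rotations.
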(graph Create_Spline_Shape_CoreProcess
  (params (Position :type Vector) (Rotation :type Rotator) (Scale :type Vector :default (1.0 1.0 1.0)) (Width :type Double :default 1e+03) (Length :type Double :default 1e+03) (Ngon :type Int :default 4) (Linear :type Bool :default true) (AtWorldOrigin :type Bool))
  (bind RemoveEmptyData_0 (Remove%20Empty%20Data))
  (bind DataNum_1 (Data%20Count))
  (bind Branch_0 (Branch))
  (bind Branch_4 (Branch :bOutputToB-param AtWorldOrigin))
  (bind GetActorData_1 (Get%20Actor%20Data :Mode GetSinglePoint))
  (bind MakeTransformAttribute_1 (Make%20Transform%20Attribute :InputSource1 $Transform.position :InputSource2 $Transform.rotation :InputSource3 mult :OutputTarget $Transform))
  (bind AttributeTransformOp_2 (Attribute%20Transform%20Op :InputSource1 $Transform :InputSource2 $Transform)) ; Transform created line to local space
  (bind FilterDataByType_2 (Filter%20Data%20By%20Type :TargetType (struct :Ids [(struct :Struct /Script/PCG.PCGDataTypeInfoPoint)])))
  (bind MergePoints_1 (Merge%20Points :bMergeMetadata false :MergeTagsMode None))
  (bind CopyAttributes_0 (Copy%20Attributes :InputSource $Index :OutputTarget tempindex))
  (bind CopyPoints_0 (Copy%20Points :bCopyEachSourceOnEveryTarget false :TagInheritance Source))
  (bind AttributePartition_2 (Attribute%20Partition :PartitionAttributeSelectors [tempindex]))
  (bind DeleteAttributes_1 (Delete%20Attributes :SelectedAttributes "tempindex"))
  (bind CreateSpline_1 (Create%20Spline :bClosedLoop true :bLinear true :InterpTypeAttribute None :bLinear-param Linear)) ; Create spline shape data
  (bind SplineDirection_0 (Spline%20Direction :Operation ForceCounterClockwise)) ; force winding order
  ;[0 0 0 0.5] Shape creation
  (bind MakeTransformAttribute_0 (Make%20Transform%20Attribute :InputSource3 scale :Rotation Rotation :Translation Position))
  (bind AttributeMathsOp_1 (Attribute%20Maths%20Op :Operation Multiply :InA Width))
  (bind AttributeCompareOp_0 (Attribute%20Compare%20Op :InA Ngon))
  (bind Branch_1 (Branch))
  (bind CreatePointsSphere_0 (Create%20Points%20Sphere :LatitudinalEndAngle 0.0 :LatitudinalSegments 1 :LatitudinalStartAngle 0.0 :LongitudinalSegments 7 :Radius 663.616177 :SphereGeneration Segments :LongitudinalSegments-param Ngon)) ; Creating the shape
  (bind AttributeMathsOp_0 (Attribute%20Maths%20Op :InputSource1 $Rotation.Roll :Operation Set)) ; removing the vertical nature of the points, \nbecause technically taken from the side of a sphere
  (bind AttributeMathsOp_2 (Attribute%20Maths%20Op :Operation Divide :InA Length :InB Width))
  (bind CreateConstant_0 (Create%20Constant :AttributeTypes (struct :Type Float :FloatValue 1.0)))
  (bind MakeVectorAttribute_1 (Make%20Vector%20Attribute :OutputTarget Scale :OutputType Vector))
  (bind MakeTransformAttribute_2 (Make%20Transform%20Attribute))
  (bind AttributeTransformOp_3 (Attribute%20Transform%20Op :InputSource1 $Transform))
  (bind AttributeTransformOp_0 (Attribute%20Transform%20Op :InputSource1 $Transform))
  ;;[0 0 0 0.5] Square shape specific case
  (bind MakeVectorAttribute_0 (Make%20Vector%20Attribute :OutputType Vector :X Width :Y Length))
  (bind AttributeMathsOp_3 (Attribute%20Maths%20Op :Operation Multiply))
  (bind CreatePointsGrid_0 (Create%20Points%20Grid :bSetPointsBounds false :PointPosition CellCorners))
  (bind FindConvexHull2D_0 (Find%20Convex%20Hull%202D))
  ;[0 0 0 0.5] Since we create the square from a grid it doesn't get proper local rotation values, Rebuilding the spline then ensures proper rotation values are stored.
  (bind AttributeCompareOp_1 (Attribute%20Compare%20Op :InA Ngon))
  (bind Branch_2 (Branch))
  (bind GetSplineControlPoints_0 (Get%20Spline%20Control%20Points))
  (bind CreateSpline_0 (Create%20Spline :bClosedLoop true :bLinear true :InterpTypeAttribute None))
  (reroute Parent_actor%20transform MakeTransformAttribute_1)
  (reroute Transform MakeTransformAttribute_0)
  (connect CreatePointsSphere_0 Out AttributeMathsOp_0 InA)
  (connect MakeVectorAttribute_0 Out AttributeMathsOp_3 InA)
  (connect CopyPoints_0 Out AttributePartition_2 In)
  (connect AttributeTransformOp_3 Out AttributeTransformOp_0 InA)
  (connect FindConvexHull2D_0 Out AttributeTransformOp_0 InA)
  (connect MakeTransformAttribute_0 Out AttributeTransformOp_0 InB)
  (connect Branch_4 Output%20A AttributeTransformOp_2 InA)
  (connect MakeTransformAttribute_1 Out AttributeTransformOp_2 InB)
  (connect AttributeMathsOp_0 Out AttributeTransformOp_3 InA)
  (connect MakeTransformAttribute_2 Out AttributeTransformOp_3 InB)
  (connect DataNum_1 Out Branch_0 bOutputToB)
  (connect AttributeTransformOp_0 Out Branch_0 In)
  (connect AttributeCompareOp_0 Out Branch_1 bOutputToB)
  (connect AttributeCompareOp_0 Out Branch_1 In)
  (connect AttributeCompareOp_1 Out Branch_2 bOutputToB)
  (connect SplineDirection_0 Out Branch_2 In)
  (connect Branch_0 Output%20A Branch_4 In)
  (connect MergePoints_1 Out CopyAttributes_0 Source)
  (connect MergePoints_1 Out CopyAttributes_0 Target)
  (connect Branch_0 Output%20B CopyPoints_0 Source)
  (connect CopyAttributes_0 Out CopyPoints_0 Target)
  (connect MakeVectorAttribute_0 Out CreatePointsGrid_0 CellSize)
  (connect Branch_1 Output%20B CreatePointsGrid_0 Execution%20Dependency)
  (connect AttributeMathsOp_3 Out CreatePointsGrid_0 GridExtents)
  (connect Branch_1 Output%20A CreatePointsSphere_0 Execution%20Dependency)
  (connect AttributeMathsOp_1 Out CreatePointsSphere_0 Radius)
  (connect GetSplineControlPoints_0 Out CreateSpline_0 In)
  (connect AttributeTransformOp_2 Out CreateSpline_1 In)
  (connect Branch_4 Output%20B CreateSpline_1 In)
  (connect DeleteAttributes_1 Out CreateSpline_1 In)
  (connect RemoveEmptyData_0 Out DataNum_1 In)
  (connect AttributePartition_2 Out DeleteAttributes_1 In)
  (connect DefaultInputNode In FilterDataByType_2 In)
  (connect CreatePointsGrid_0 Out FindConvexHull2D_0 In)
  (connect Branch_2 Output%20B GetSplineControlPoints_0 In)
  (connect GetActorData_1 Out MakeTransformAttribute_1 Rotation)
  (connect GetActorData_1 Out MakeTransformAttribute_1 Translation)
  (connect MakeVectorAttribute_1 Out MakeTransformAttribute_2 Scale)
  (connect CreateConstant_0 Out MakeVectorAttribute_1 X)
  (connect AttributeMathsOp_2 Out MakeVectorAttribute_1 Y)
  (connect CreateConstant_0 Out MakeVectorAttribute_1 Z)
  (connect FilterDataByType_2 InsideFilter MergePoints_1 In)
  (connect DefaultInputNode In RemoveEmptyData_0 In)
  (connect CreateSpline_1 Out SplineDirection_0 In)
  (output Out Branch_2.Output%20A)
  (output Out CreateSpline_0))
