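; This PCG subgraph is the core loop body for creating spline shapes, turning a set of shape points into a closed outline at target points.
; It filters the main input down to point data and copies the shape input points onto those targets.
; The copied points feed a Create Spline node, which outputs closed-loop splines with linear segments.
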(graph Create_Spline_Shape_CoreProcessLoop
  (inputs (shape%20input :type Point))
  (bind FilterDataByType_1 (Filter%20Data%20By%20Type :TargetType (struct :Ids [(struct :Struct /Script/PCG.PCGDataTypeInfoPoint)])))
  (bind CopyPoints_0 (Copy%20Points))
  (bind CreateSpline_0 (Create%20Spline :bClosedLoop true :bLinear true :InterpTypeAttribute None))
  (connect DefaultInputNode shape%20input CopyPoints_0 Source)
  (connect FilterDataByType_1 InsideFilter CopyPoints_0 Target)
  (connect CopyPoints_0 Out CreateSpline_0 In)
  (connect DefaultInputNode In FilterDataByType_1 In)
  (output Out CreateSpline_0))
