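; PCG subgraph that creates straight linear spline lines from Length, Position and Rotation parameters.
; Two end points are scaled to the length and offset by copying onto a transformed anchor point.
; With input data, the line is copied onto each input point, then partitioned by a temporary index into separate splines.
; Without input, AtWorldOrigin chooses between world space and the actor's transform before Create Spline.
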(graph Create_Spline_line_CoreProcess
  (params (Position :type Vector) (Rotation :type Rotator) (Length :type Double :default 1e+03) (AtWorldOrigin :type Bool :default true))
  (bind RemoveEmptyData_0 (Remove%20Empty%20Data))
  (bind DataNum_1 (Data%20Count))
  (bind Branch_1 (Branch)) ; If primary input data is provided,\ncopy created line over each input point/controlpoint
  (bind Branch_2 (Branch :bOutputToB-param AtWorldOrigin))
  (bind GetActorData_0 (Get%20Actor%20Data :Mode GetSinglePoint))
  (bind MakeTransformAttribute_2 (Make%20Transform%20Attribute :InputSource1 $Transform.position :InputSource2 $Transform.rotation :InputSource3 mult :OutputTarget $Transform))
  (bind AttributeTransformOp_0 (Attribute%20Transform%20Op :InputSource1 $Transform :InputSource2 $Transform)) ; Transform created line to local space
  (bind FilterDataByType_1 (Filter%20Data%20By%20Type :TargetType (struct :Ids [(struct :Struct /Script/PCG.PCGDataTypeInfoPoint)])))
  (bind MergePoints_0 (Merge%20Points :bMergeMetadata false :MergeTagsMode None))
  (bind CopyAttributes_1 (Copy%20Attributes :InputSource $Index :OutputTarget tempindex))
  (bind CopyPoints_2 (Copy%20Points :bCopyEachSourceOnEveryTarget false :TagInheritance Source))
  (bind AttributePartition_1 (Attribute%20Partition :PartitionAttributeSelectors [tempindex]))
  (bind DeleteAttributes_0 (Delete%20Attributes :SelectedAttributes "tempindex"))
  (bind CreateSpline_1 (Create%20Spline :bLinear true :InterpTypeAttribute None)) ; Create spline line data
  ;[0 0 0 0.5] Create line start and end points, set length and apply offsets
  (bind AttributeMathsOp_5 (Attribute%20Maths%20Op :Operation Max :InA Length))
  (bind CreatePoints_1 (Create%20Points :PointsToCreate [(struct :Transform (struct :Translation (-0.5 0.0 0.0)) :Steepness 1.0) (struct :Transform (struct :Translation (0.5 0.0 0.0)) :Steepness 1.0)]))
  (bind AttributeMathsOp_4 (Attribute%20Maths%20Op :InputSource1 $Position.x :Operation Multiply))
  (bind MakeTransformAttribute_1 (Make%20Transform%20Attribute :InputSource3 mult :Rotation Rotation :Translation Position))
  (bind CreatePoints_2 (Create%20Points :PointsToCreate [(struct :Transform (struct :Translation (-0.5 0.0 0.0)) :Steepness 1.0)]))
  (bind CopyPoints_1 (Copy%20Points :AttributeInheritance SourceOnly :TagInheritance Source))
  (connect CreatePoints_1 Out AttributeMathsOp_4 InA)
  (connect AttributeMathsOp_5 Out AttributeMathsOp_4 InB)
  (connect CopyPoints_2 Out AttributePartition_1 In)
  (connect Branch_2 Output%20A AttributeTransformOp_0 InA)
  (connect MakeTransformAttribute_2 Out AttributeTransformOp_0 InB)
  (connect DataNum_1 Out Branch_1 bOutputToB)
  (connect CopyPoints_1 Out Branch_1 In)
  (connect Branch_1 Output%20A Branch_2 In)
  (connect MergePoints_0 Out CopyAttributes_1 Source)
  (connect MergePoints_0 Out CopyAttributes_1 Target)
  (connect AttributeMathsOp_4 Out CopyPoints_1 Source)
  (connect CreatePoints_2 Out CopyPoints_1 Target)
  (connect Branch_1 Output%20B CopyPoints_2 Source)
  (connect CopyAttributes_1 Out CopyPoints_2 Target)
  (connect MakeTransformAttribute_1 Out CreatePoints_2 Transform)
  (connect AttributeTransformOp_0 Out CreateSpline_1 In)
  (connect Branch_2 Output%20B CreateSpline_1 In)
  (connect DeleteAttributes_0 Out CreateSpline_1 In)
  (connect RemoveEmptyData_0 Out DataNum_1 In)
  (connect AttributePartition_1 Out DeleteAttributes_0 In)
  (connect DefaultInputNode In FilterDataByType_1 In)
  (connect GetActorData_0 Out MakeTransformAttribute_2 Rotation)
  (connect GetActorData_0 Out MakeTransformAttribute_2 Translation)
  (connect FilterDataByType_1 InsideFilter MergePoints_0 In)
  (connect DefaultInputNode In RemoveEmptyData_0 In)
  (output Out CreateSpline_1))
